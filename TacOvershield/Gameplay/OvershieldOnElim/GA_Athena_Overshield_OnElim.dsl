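; GA_Athena_Overshield_OnElim is an event-triggered gameplay ability that gives the player overshield on an elimination.
; It responds to the event only when the spreadsheet row for overshield heal-on-kill evaluates to true, so the feature is toggled through data.
; On activation it applies the OvershieldOnElim gameplay effect to its owner and prints a debug string.
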(event Ability|EventActivateAbilityFromEvent
  (params (EventData :type Gameplay%20Event%20Data%20Structure%20%28by%20ref%29))
  ;[1 1 1] On Activate - Give overshield on Elim to the player.
  (Ability|ApplyGameplayEffectToOwner (Variables|Default|GetGE_OvershieldOnElim))
  (Development|PrintString "OvershieldOnElim"))

(fn ShouldAbilityRespondToEvent
  (params (ActorInfo :type Gameplay%20Ability%20Actor%20Info%20Structure) (Payload :type Gameplay%20Event%20Data%20Structure))
  ;[1 1 1] Only Trigger If Spreedsheet row is is enabled
  (if (Class|AthenaFunctionLibrary|RowtoBool (Variables|Default|GetRow_OvershieldHealOnKillEnabled))
    (return true)
    (else
      (return false))))
